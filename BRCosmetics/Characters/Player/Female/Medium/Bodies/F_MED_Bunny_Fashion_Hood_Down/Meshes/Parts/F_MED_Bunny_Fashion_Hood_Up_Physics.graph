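# Physics asset for the Hood Up mesh part of the female medium Bunny Fashion character, covering the head and its dynamic hair bones.
# A large head capsule anchors four smaller hair capsules: two front strands, one right strand and a front hair piece.
# Each hair bone constrains directly to the head with tight twist and limited swing, the front piece having the narrowest swing.
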PhysicsAsset /BRCosmetics/Characters/Player/Female/Medium/Bodies/F_MED_Bunny_Fashion_Hood_Down/Meshes/Parts/F_MED_Bunny_Fashion_Hood_Up_Physics
bodies {
 dyn_fr_hair_01 [{shapeType=Capsule, length=0.505, radius=0.505, center=<1.5 0 0>, rotation=<rot 90 0 0>}]
 dyn_fr_hair_02 [{shapeType=Capsule, length=0.505, radius=0.505, center=<1.5 0 0>, rotation=<rot 90 0 0>}]
 dyn_hair_front [{shapeType=Capsule, length=2.2291, radius=1.9199, center=<2 0 0>, rotation=<rot 90 0 0>}]
 dyn_rt_hair_01 [{shapeType=Capsule, length=0.505, radius=0.505, center=<1.5 0 0>, rotation=<rot 90 0 0>}]
 head [{shapeType=Capsule, length=5.9394, radius=6.1271, center=<4.2209 1.4624 0.1736>, rotation=<rot 90 0 0>}]
}
constraints [
 {bone1Name=dyn_fr_hair_02, bone2Name=head, swing1LimitDegrees=25, swing2LimitDegrees=25, twistLimitDegrees=5}
 {bone1Name=dyn_rt_hair_01, bone2Name=head, swing1LimitDegrees=25, swing2LimitDegrees=25, twistLimitDegrees=5}
 {bone1Name=dyn_fr_hair_01, bone2Name=head, swing1LimitDegrees=25, swing2LimitDegrees=25, twistLimitDegrees=5}
 {bone1Name=dyn_hair_front, bone2Name=head, swing1LimitDegrees=10, swing2LimitDegrees=10, twistLimitDegrees=5}
]
